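; Blueprint child of a time-of-day manager, apparently a static skylight variant, that inherits its day-night cycle behavior from a parent class.
; It adds a CycloneTODM custom event, used by the Cyclone event, which hides the sky dome mesh component without propagating to its children.
; BeginPlay and the construction script only call the parent, and the Tick override is an unused, disabled stub.
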(event Custom|CycloneTODM
  ;[1 1 1] This is used by the Cyclone event
  (Rendering|SetVisibility (Variables|DayNightCycle|GetSkyDomeMeshComp) false false))

(event EventBeginPlay
  (|CallParentFunction))

(event EventTick
  (params (DeltaSeconds :type Float%20%28single-precision%29))
  (|CallParentFunction DeltaSeconds) ; This node is disabled and will not be called.\nDrag off pins to build functionality.
)

(fn ConstructionScript
  (|CallParentFunction))
